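; Chrome shotgun weapon Blueprint that drives crystal and upgrade visuals as the weapon evolves through XP.
; BeginPlay fires a local reform-crystals GameplayCue for the locally controlled pawn, plays the upgrade animation, and binds to the Chrome XP component.
; Material effects use Custom Primitive Data on the weapon mesh: XP progress, rarity-based values, and timestamps for upgrade and anim-notify crystal shatter.
; The power-up-ready and custom fire-audio events are empty.
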(event EventBeginPlay
  ;[0.1158 1 0.1046] Begin Play - bind to callbacks and initialize.
  (|CallParentFunction)
  (bind _asfort_player_pawn (Utilities|Casting|CastToFortPlayerPawn (Game|GetInstigator))
    (:then
      (if (Pawn|IsLocallyControlled _asfort_player_pawn)
        (GameplayCue|ExecuteGameplayCue%28Burst%29Local (Game|GetInstigator) (Variables|Default|GetGCN_ReformCrystals) (Ability|GameplayCue|MakeGameplayCueParameters 0.0))
        (else
          _))
      (MID|MIDPlayUpgradeAnimation)
      (bind _output_get (Variables|Default|SetChromeXPComp (Actor|GetComponentbyClass "/DistortedWeaponsGameplay/Chrome/C_ChromeWeapons_XPComponent.C_ChromeWeapons_XPComponent_C")))
      (Utilities|IsValid _output_get
        (:Is%20Valid
          (Default|BindEventtoOnPowerUpReady _output_get (EventDispatchers|CreateEvent :!event Custom|OnPowerUpReady))
          (Variables|Default|SetEvolutionLevel (MultiItem|GetSelectedItemIndex _output_get))
          (Default|BindEventtoOnXPChanged _output_get (EventDispatchers|CreateEvent :!event Custom|On%20XP%20Change)))))
    (:CastFailed
      _)))

(event Custom|OnPowerUpReady
  (params (CurrentItemIndex :type Integer) (PendingItemIndex :type Integer)))

(event EventPlayWeaponFireCustomAudio
  (params (bPersistentFire :type Boolean) (bSecondaryFire :type Boolean)))

(event Custom|AnimNotify_Upgrade_ShatterCrystals
  ;[1 1 1] Event that plays when crystals shatter in upgrade animation
  (MID|MIDShatterCrystals))

(event Custom|On%20XP%20Change
  (params (XPDelta :type Float%20%28double-precision%29) (CurrentXPPercentage :type Float%20%28double-precision%29))
  ;[1 1 1] Event that plays when crystals shatter in upgrade animation
  (MID|MIDChangeXP CurrentXPPercentage))

(fn GetAffiliationsToShowFor
  (return (Utilities|Array|MakeArray "Hostile" "Neutral")))

(fn MIDChangeXP
  (params (XP%20Normalized :type Float%20%28double-precision%29))
  (Rendering|Material|SetCustomPrimitiveDataFloat (Variables|Visual|GetWeaponMesh) 0 XP%20Normalized))

(fn MIDPlayUpgradeAnimation
  (bind _returnvalue (Rarity|GetWeaponDataRarity))
  (bind _asfort_player_pawn (Utilities|Casting|CastToFortPlayerPawn (Game|GetInstigator))
    (:then
      (if (Pawn|IsLocallyControlled _asfort_player_pawn)
        (Rendering|Material|SetCustomPrimitiveDataFloat (Variables|Visual|GetWeaponMesh) 2 (Utilities|Time|GetTimeSeconds))
        (else
          _))
      (Rendering|Material|SetCustomPrimitiveDataFloat (Variables|Visual|GetWeaponMesh) 1 (Utilities|Select :Index _returnvalue :Common 0.0 :Uncommon 0.0 :Rare 1.0 :Epic 2.0 :Legendary 2.0 :Mythic 2.0 :Transcendent 2.0 :Unattainable 2.0))
      (Rendering|Material|SetCustomPrimitiveDataFloat (Variables|Visual|GetWeaponMesh) 3 (Math|Conversions|ToFloat(Boolean) (Utilities|Select :Index _returnvalue :Common false :Uncommon false :Rare false :Epic false :Legendary true :Mythic true :Transcendent true :Unattainable true))))
    (:CastFailed
      _)))

(fn MIDShatterCrystals
  (Rendering|Material|SetCustomPrimitiveDataFloat (Variables|Visual|GetWeaponMesh) 4 (Utilities|Time|GetTimeSeconds)))

(fn OnRep_EvolutionLevel)

(fn ConstructionScript
  (|CallParentFunction))
