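; Tracer_Chrome_Shotgun is the tracer effect Blueprint for the Chrome shotgun, built around a tracer mesh component.
; Its construction script sets that mesh's relative rotation to zero, without sweep or teleport.
; BeginPlay is present but empty, so the graph adds no runtime behaviour of its own.
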(event EventBeginPlay)

(fn ConstructionScript
  (Transformation|SetRelativeRotation (Variables|Visual|GetTracerMesh) "0, 0, 0" false false))
