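# Glowing cyan-to-green material for Niagara mesh particles, named for the shard pieces of the Chrome shotgun; it is a base Material despite the MI_ prefix.
# A sphere mask around each particle's position, sized by object scale, blends the two colors and masks the glow.
# That mask plus a view-angle dot product drives hue shift before MF_FX_EmissiveAdjustments.
# Dynamic parameters control mask size, glow strength and hue shift.
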Material /DistortedWeaponsGameplay/Chrome/Shotgun/Materials/MI_Chrome_ShardPieces
graph {
 properties {bUsedWithNiagaraMeshParticles=true, pixelDepthOffsetMode=PDOM_Legacy, usedShadingModels=Default Lit}
 nodes [
  n1 = DynamicParameter() {defaultValue=<rgba 25 1 0.5 1>, paramNames [ObjectBoundsScaleMult, GlowStrenght, HueShiftStrenght, Param4]}
  n2 = SphereMask(
   A=ParticlePositionWS()
   B=WorldPosition().XYZ
   Radius=Multiply(
    A=@n1.ObjectBoundsScaleMult
    B=MaterialFunctionCall() {materialFunction=&ObjectScale}.Scale XYZ
   )
   Hardness=Constant() {r=-1}
  )
  n3 = DotProduct(A=VertexNormalWS(), B=CameraVectorWS())
  n4 = DynamicParameter() {defaultValue=<rgba 25 1 0.5 1>, paramNames [ObjectBoundsScaleMult, GlowStrenght, HueShiftStrenght, Param4]}
  n5 = MaterialFunctionCall(
   Hue Shift Percentage=Multiply(A=Add(A=@n2, B=@n3), B=@n4.HueShiftStrenght) {constB=0.5}
   Texture=LinearInterpolate(A=Constant3Vector() {constant=<rgba 0 0.8098 1 0>}, B=Constant3Vector() {constant=<rgba 0 1 0.1238 0>}, Alpha=@n2)
  ) {materialFunction=&HueShift}
  MaterialFunctionCall(Emissive=LinearInterpolate(B=Multiply(A=@n5.Result, B=@n4.GlowStrenght) {constB=5}, Alpha=@n2) {constB=0}) {materialFunction=&MF_FX_EmissiveAdjustments}
  ObjectPositionWS()
 ]
}
