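# Physics asset for the Chrome Shotgun weapon mesh, giving it simple collision and ragdoll-style simulation.
# It has two capsule bodies: a larger one on the root bone and a smaller one on the handle bone.
# A single constraint links handle to root with equal angular limits on both swing axes and twist.
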PhysicsAsset /DistortedWeaponsGameplay/Chrome/Shotgun/Meshes/Chrome_Shotgun_PhysicsAsset
bodies {
 handle [{shapeType=Capsule, length=23.8475, radius=5.3773, center=<28.2816 0.0387 4.6092>, rotation=<rot 0 -90.1717 -84.7432>}]
 root [{shapeType=Capsule, length=46.5835, radius=12.2742, center=<0.0106 20.4128 3.1998>, rotation=<rot 0 -179.8712 -88.6568>}]
}
constraints [{bone1Name=handle, bone2Name=root, swing1LimitDegrees=45, swing2LimitDegrees=45, twistLimitDegrees=45}]
