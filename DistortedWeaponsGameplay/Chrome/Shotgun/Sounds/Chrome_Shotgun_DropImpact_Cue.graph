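# Sound cue for the impact sound of a dropped Chrome shotgun.
# A single wave player feeds a modulator that slightly randomizes pitch and volume on each play, with the cue boosting overall volume.
# It uses the loot-collect sound class, overrides attenuation for a short falloff range, and limits overlapping plays through the shared weapon-drop stop-oldest concurrency set.
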SoundCue /DistortedWeaponsGameplay/Chrome/Shotgun/Sounds/Chrome_Shotgun_DropImpact_Cue
root=Modulator_0
nodes [
 {id=Modulator_0, children [WavePlayer_0], properties {pitchMax=1.1, pitchMin=1, volumeMax=1.1, volumeMin=0.9}}
 {id=WavePlayer_0, properties {soundWaveAssetPtr=&/Chrome_Shotgun_DropImpact_01}}
]
properties {
 bOverrideAttenuation=true
 maxDistance=1000
 soundClassObject=&Fort_Loot_Collect
 volumeMultiplier=1.7
 attenuationOverrides {falloffDistance=1000, attenuationShapeExtents {x=0}}
 concurrencyOverrides {maxCount=3, resolutionRule=PreventNew}
 concurrencySet [&Weapon_Drop_StopOldest]
}
