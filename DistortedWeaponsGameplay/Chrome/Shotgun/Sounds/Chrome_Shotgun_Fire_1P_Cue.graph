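# Sound cue for the first-person close firing sound of the Chrome shotgun.
# A quality-level switch sends the first level to a random pick of four close-fire waves, while the other three share a random pick of two.
# Each branch applies slight upward random pitch modulation.
# The cue uses first-person weapon attenuation, a ranged-fire sound class and a first-person gunfire concurrency set.
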SoundCue /DistortedWeaponsGameplay/Chrome/Shotgun/Sounds/Chrome_Shotgun_Fire_1P_Cue
root=QualityLevel_0
nodes [
 {id=Modulator_0, children [Random_0], properties {pitchMax=1.1, pitchMin=1}}
 {id=Modulator_1, children [Random_1], properties {pitchMax=1.1, pitchMin=1}}
 {id=QualityLevel_0, children [Modulator_0,Modulator_1,Modulator_1,Modulator_1]}
 {id=Random_0, children [WavePlayer_0,WavePlayer_1,WavePlayer_2,WavePlayer_7], properties {weights [1,1,1,1]}}
 {id=Random_1, children [WavePlayer_3,WavePlayer_4], properties {weights [1,1]}}
 {id=WavePlayer_0, properties {soundWaveAssetPtr=&/Chrome_Shotgun_Fire_Close_01}}
 {id=WavePlayer_1, properties {soundWaveAssetPtr=&/Chrome_Shotgun_Fire_Close_02}}
 {id=WavePlayer_2, properties {soundWaveAssetPtr=&/Chrome_Shotgun_Fire_Close_03}}
 {id=WavePlayer_3, properties {soundWaveAssetPtr=&Chrome_Shotgun_Fire_Close_01}}
 {id=WavePlayer_4, properties {soundWaveAssetPtr=&Chrome_Shotgun_Fire_Close_02}}
 {id=WavePlayer_7, properties {soundWaveAssetPtr=&/Chrome_Shotgun_Fire_Close_04}}
]
properties {
 attenuationSettings=&Fort_Weapon_Projectile_FirstPerson_Athena_Att
 maxDistance=2097152
 soundClassObject=&Fort_WeaponRanged_Fire_1P_PSM
 volumeMultiplier=0.7
 attenuationOverrides {falloffDistance=5000, attenuationShapeExtents {x=1000}}
 concurrencyOverrides {bLimitToOwner=true, maxCount=3}
 concurrencySet [&Weapon_Gunfire_FirstPerson_Conc]
}
