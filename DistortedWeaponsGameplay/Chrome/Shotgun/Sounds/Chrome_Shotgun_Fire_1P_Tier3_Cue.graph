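# Sound cue for the first-person fire sound of the Tier 3 Chrome shotgun.
# A quality-level switch feeds the first level a random pick of four close-fire mixes, while the other three levels share a pick of two.
# Both branches apply slight upward pitch modulation for variety.
# The cue uses the first-person ranged-fire sound class, first-person projectile attenuation and a Chrome tier-fire concurrency asset from the Assault folder.
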SoundCue /DistortedWeaponsGameplay/Chrome/Shotgun/Sounds/Chrome_Shotgun_Fire_1P_Tier3_Cue
root=QualityLevel_0
nodes [
 {id=Modulator_0, children [Random_0], properties {pitchMax=1.1, pitchMin=1}}
 {id=Modulator_2, children [Random_2], properties {pitchMax=1.1, pitchMin=1}}
 {id=QualityLevel_0, children [Modulator_2,Modulator_0,Modulator_0,Modulator_0]}
 {id=Random_0, children [WavePlayer_0,WavePlayer_1], properties {weights [1,1]}}
 {id=Random_2, children [WavePlayer_9, WavePlayer_10, WavePlayer_11, WavePlayer_12], properties {weights [1,1,1,1]}}
 {id=WavePlayer_0, properties {soundWaveAssetPtr=&/Chrome_Shotgun_Fire_Close_Tier03_Mix_01}}
 {id=WavePlayer_1, properties {soundWaveAssetPtr=&/Chrome_Shotgun_Fire_Close_Tier03_Mix_02}}
 {id=WavePlayer_10, properties {soundWaveAssetPtr=&Chrome_Shotgun_Fire_Close_Tier03_Mix_02}}
 {id=WavePlayer_11, properties {soundWaveAssetPtr=&/Chrome_Shotgun_Fire_Close_Tier03_Mix_03}}
 {id=WavePlayer_12, properties {soundWaveAssetPtr=&/Chrome_Shotgun_Fire_Close_Tier03_Mix_04}}
 {id=WavePlayer_9, properties {soundWaveAssetPtr=&Chrome_Shotgun_Fire_Close_Tier03_Mix_01}}
]
properties {
 attenuationSettings=&Fort_Weapon_Projectile_FirstPerson_Athena_Att
 maxDistance=2097152
 soundClassObject=&Fort_WeaponRanged_Fire_1P_PSM
 volumeMultiplier=0.8
 attenuationOverrides {falloffDistance=5000, attenuationShapeExtents {x=1000}}
 concurrencyOverrides {bLimitToOwner=true, maxCount=2, resolutionRule=StopOldest, voiceStealReleaseTime=0.3, volumeScaleAttackTime=0.3}
 concurrencySet [/DistortedWeaponsGameplay/Chrome/Assault/Sounds/Chrome_Weapon_Tier_Fire, null]
}
