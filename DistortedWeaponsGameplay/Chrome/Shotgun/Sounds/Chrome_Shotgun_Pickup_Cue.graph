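# Sound cue for picking up the Chrome Shotgun in the DistortedWeapons gameplay content.
# A single wave player feeds a Modulator that slightly randomizes volume on each play.
# It uses the shared weapon pickup attenuation and the loot-collect sound class.
# A shared pickup concurrency set is assigned, alongside override values for a single prevent-new instance.
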SoundCue /DistortedWeaponsGameplay/Chrome/Shotgun/Sounds/Chrome_Shotgun_Pickup_Cue
root=Modulator_0
nodes [
 {id=Modulator_0, children [WavePlayer_0], properties {volumeMax=1.1, volumeMin=0.9}}
 {id=WavePlayer_0, properties {soundWaveAssetPtr=&/Chrome_Shotgun_Pickup_01}}
]
properties {
 attenuationSettings=&Fort_Weapon_Pickup_Att
 maxDistance=3500
 soundClassObject=&Fort_Loot_Collect
 volumeMultiplier=1.1077
 concurrencyOverrides {maxCount=1, resolutionRule=PreventNew}
 concurrencySet [&PickupMaximum_Oldest]
}
