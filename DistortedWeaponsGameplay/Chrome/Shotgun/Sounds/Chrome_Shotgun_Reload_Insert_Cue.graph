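# Sound cue for the shell-insert foley of the Chrome Shotgun's reload.
# A Random node picks one of five equally weighted insert waves, and a Modulator applies slight random pitch variation to each play.
# It routes through the weapon foley sound class with weapon reload attenuation and uses shared local and global basic foley concurrency sets.
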SoundCue /DistortedWeaponsGameplay/Chrome/Shotgun/Sounds/Chrome_Shotgun_Reload_Insert_Cue
root=Modulator_13
nodes [
 {id=Modulator_13, children [Random_1], properties {pitchMax=1.1, pitchMin=0.9}}
 {id=Random_1, children [WavePlayer_3,WavePlayer_4,WavePlayer_5,WavePlayer_6,WavePlayer_7], properties {weights [1,1,1,1,1]}}
 {id=WavePlayer_3, properties {soundWaveAssetPtr=&/Chrome_Shotgun_Reload_Insert_01}}
 {id=WavePlayer_4, properties {soundWaveAssetPtr=&/Chrome_Shotgun_Reload_Insert_02}}
 {id=WavePlayer_5, properties {soundWaveAssetPtr=&/Chrome_Shotgun_Reload_Insert_03}}
 {id=WavePlayer_6, properties {soundWaveAssetPtr=&/Chrome_Shotgun_Reload_Insert_04}}
 {id=WavePlayer_7, properties {soundWaveAssetPtr=&/Chrome_Shotgun_Reload_Insert_05}}
]
properties {
 attenuationSettings=&Fort_Weapon_Reload_Att
 maxDistance=2250
 soundClassObject=&Fort_WeaponFoley
 volumeMultiplier=0.802
 attenuationOverrides {falloffDistance=1600}
 concurrencyOverrides {resolutionRule=PreventNew}
 concurrencySet [
  /Game/Sounds/Fort_Foley_Sounds/FoleyLibraries/Settings/Foley_Basic_Local_Conc
  /Game/Sounds/Fort_Foley_Sounds/FoleyLibraries/Settings/Foley_Basic_Global_Conc
 ]
}
