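# Sound cue for the upgraded Chrome Shotgun's reload foley, variant B.
# A single wave player feeds the root with one reload sample at reduced volume, with no randomization or modulation.
# It is classed as weapon foley, references the shared weapon reload attenuation, and joins the basic local and global foley concurrency sets.
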SoundCue /DistortedWeaponsGameplay/Chrome/Shotgun/Sounds/Chrome_Shotgun_Upgrade_Reload_B_Cue
root=WavePlayer_0
nodes [{id=WavePlayer_0, properties {soundWaveAssetPtr=&/Chrome_Shotgun_Upgrade_Reload_B_01}}]
properties {
 attenuationSettings=&Fort_Weapon_Reload_Att
 maxDistance=2250
 soundClassObject=&Fort_WeaponFoley
 volumeMultiplier=0.501
 attenuationOverrides {falloffDistance=1600}
 concurrencyOverrides {resolutionRule=PreventNew}
 concurrencySet [
  /Game/Sounds/Fort_Foley_Sounds/FoleyLibraries/Settings/Foley_Basic_Local_Conc
  /Game/Sounds/Fort_Foley_Sounds/FoleyLibraries/Settings/Foley_Basic_Global_Conc
 ]
}
