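# Sound cue for the close bullet whip-by of the Chrome Shotgun, in the projectile effects sound class.
# A quality-level root sends its first two slots to a random pick of six whip waves and its last two to a pick of two.
# Each branch passes through a modulator that randomizes pitch and volume.
# The cue overrides attenuation to disable distance falloff and uses a shared shotgun whiz-by concurrency rule.
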SoundCue /DistortedWeaponsGameplay/Chrome/Shotgun/Sounds/Chrome_Shotgun_Whip_Close_Cue
root=QualityLevel_1
nodes [
 {id=Modulator_2, children [Random_0], properties {pitchMax=1.4, pitchMin=0.9, volumeMax=2.3, volumeMin=1.7}}
 {id=Modulator_3, children [Random_1], properties {pitchMax=1.4, pitchMin=0.9, volumeMax=2.3, volumeMin=1.7}}
 {id=QualityLevel_1, children [Modulator_2,Modulator_2,Modulator_3,Modulator_3]}
 {id=Random_0, children [WavePlayer_0,WavePlayer_1,WavePlayer_2,WavePlayer_3,WavePlayer_4,WavePlayer_6], properties {weights [1,1,1,1,1,1]}}
 {id=Random_1, children [WavePlayer_11, WavePlayer_12], properties {weights [1,1]}}
 {id=WavePlayer_0, properties {soundWaveAssetPtr=&/Chrome_Shotgun_Whip_Close_01}}
 {id=WavePlayer_1, properties {soundWaveAssetPtr=&/Chrome_Shotgun_Whip_Close_02}}
 {id=WavePlayer_11, properties {soundWaveAssetPtr=&Chrome_Shotgun_Whip_Close_01}}
 {id=WavePlayer_12, properties {soundWaveAssetPtr=&Chrome_Shotgun_Whip_Close_02}}
 {id=WavePlayer_2, properties {soundWaveAssetPtr=&/Chrome_Shotgun_Whip_Close_03}}
 {id=WavePlayer_3, properties {soundWaveAssetPtr=&/Chrome_Shotgun_Whip_Close_04}}
 {id=WavePlayer_4, properties {soundWaveAssetPtr=&/Chrome_Shotgun_Whip_Close_05}}
 {id=WavePlayer_6, properties {soundWaveAssetPtr=&/Chrome_Shotgun_Whip_Close_06}}
]
properties {
 bOverrideAttenuation=true
 maxDistance=2097152
 soundClassObject=&Fort_WeaponRanged_Projectile_Effects
 volumeMultiplier=0.45
 attenuationOverrides {bAttenuate=false, stereoSpread=0}
 concurrencyOverrides {maxCount=2}
 concurrencySet [/Game/Sounds/Concurrency_Rules/Athena_BulletWhizBy_Shotgun_Conc]
}
