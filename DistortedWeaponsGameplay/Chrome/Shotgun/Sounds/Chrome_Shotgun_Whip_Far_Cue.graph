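# Sound cue for the distant bullet whip-by of the Chrome shotgun, classed as a ranged weapon projectile effect.
# A quality-level node sends its first two slots to a random pick of six shotgun whip waves, and its last two to two waves borrowed from the Chrome assault weapon.
# A root modulator randomizes pitch and volume.
# Overridden attenuation disables distance falloff, and the shotgun whiz-by concurrency set is assigned.
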SoundCue /DistortedWeaponsGameplay/Chrome/Shotgun/Sounds/Chrome_Shotgun_Whip_Far_Cue
root=Modulator_1
nodes [
 {id=Modulator_1, children [QualityLevel_1], properties {pitchMax=1.4, pitchMin=0.8, volumeMax=1.9, volumeMin=1.3}}
 {id=QualityLevel_1, children [Random_0,Random_0,Random_5,Random_5]}
 {id=Random_0, children [WavePlayer_0,WavePlayer_1,WavePlayer_2,WavePlayer_3,WavePlayer_4,WavePlayer_5], properties {weights [1,1,1,1,1,1]}}
 {id=Random_5, children [WavePlayer_22, WavePlayer_23], properties {weights [1,1]}}
 {id=WavePlayer_0, properties {soundWaveAssetPtr=&/Chrome_Shotgun_Whip_Far_01}}
 {id=WavePlayer_1, properties {soundWaveAssetPtr=&/Chrome_Shotgun_Whip_Far_02}}
 {id=WavePlayer_2, properties {soundWaveAssetPtr=&/Chrome_Shotgun_Whip_Far_03}}
 {id=WavePlayer_22, properties {soundWaveAssetPtr=/DistortedWeaponsGameplay/Chrome/Assault/Sounds/Chrome_Assault_Whip_Far_01}}
 {id=WavePlayer_23, properties {soundWaveAssetPtr=/DistortedWeaponsGameplay/Chrome/Assault/Sounds/Chrome_Assault_Whip_Far_02}}
 {id=WavePlayer_3, properties {soundWaveAssetPtr=&/Chrome_Shotgun_Whip_Far_04}}
 {id=WavePlayer_4, properties {soundWaveAssetPtr=&/Chrome_Shotgun_Whip_Far_05}}
 {id=WavePlayer_5, properties {soundWaveAssetPtr=&/Chrome_Shotgun_Whip_Far_06}}
]
properties {
 bOverrideAttenuation=true
 maxDistance=2097152
 priority=0.9
 soundClassObject=&Fort_WeaponRanged_Projectile_Effects
 volumeMultiplier=0.6
 attenuationOverrides {bAttenuate=false, stereoSpread=0}
 concurrencyOverrides {maxCount=2}
 concurrencySet [/Game/Sounds/Concurrency_Rules/Athena_BulletWhizBy_Shotgun_Conc]
}
